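; Weapon Blueprint for the Figment Burst Fire SMG that fixes shell ejection effects during burst fire.
; On BeginPlay it calls the parent and binds a custom EjectShell event to OnConsumeAmmo.
; EjectShell checks ShouldPlayShellFX and activates shell effects on each ammo consumption, because the weapon's native settings eject only one per burst (FORT-1113314).
; The Quantity parameter is unused, and Tick is empty.
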(event EventBeginPlay
  (|CallParentFunction)
  ;[1 1 1 0.5] Ensures that the Burst SMG ejects a bullet each time ammo is consumed during the weapon's burst fire (using weapon-native settings to adjust bullet ejection fx only ejected 1)
  ; (FORT-1113314)
  (EventDispatchers|BindEventtoOnConsumeAmmo (delegate Custom|EjectShell)))

(event EventTick
  (params (DeltaSeconds :type Float%20%28single-precision%29)))

(event Custom|EjectShell
  (params (Quantity :type Integer))
  ;[1 1 1 0.5] Ensures that the Burst SMG ejects a bullet each time ammo is consumed during the weapon's burst fire (using weapon-native settings to adjust bullet ejection fx only ejected 1)
  ; (FORT-1113314)
  (if (Effects|ShouldPlayShellFX)
    (CallFunction|ActivateShellsFX true)))

(fn ConstructionScript
  (|CallParentFunction))
